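# Sound cue for the Cosmos rifle's melee lunge attack.
# A random node picks one of three equally weighted lunge wave variations and is excluded from branch culling.
# It uses the shared close-range mono melee attack attenuation and the weapon melee sound class, with boosted volume.
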SoundCue /CosmosGameplay/Items/CosmosRifle/Audio/Gun_Rifle_Cosmo_Melee_Lunge_Cue
root=Random_0
nodes [
 {id=Random_0, children [WavePlayer_1,WavePlayer_2,WavePlayer_3], properties {bShouldExcludeFromBranchCulling=true, weights [1,1,1]}}
 {id=WavePlayer_1, properties {soundWaveAssetPtr=&/Gun_Rifle_Cosmo_Melee_Lunge_01}}
 {id=WavePlayer_2, properties {soundWaveAssetPtr=&/Gun_Rifle_Cosmo_Melee_Lunge_02}}
 {id=WavePlayer_3, properties {soundWaveAssetPtr=&/Gun_Rifle_Cosmo_Melee_Lunge_03}}
]
properties {
 attenuationSettings=/Game/Sounds/Attenuation/Fort_Weapon_Melee_Attack_Close_Range_Mono_Att
 maxDistance=10058.75
 soundClassObject=&Fort_WeaponMelee_PSM
 volumeMultiplier=1.25
}
