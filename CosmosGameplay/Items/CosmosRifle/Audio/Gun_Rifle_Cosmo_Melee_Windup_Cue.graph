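# Sound cue for the melee windup of the Cosmos rifle.
# A Random node with equal weights picks one of three windup wave variations.
# Playback uses close-range mono melee attack attenuation, the weapon melee sound class and a boosted volume.
# Its concurrency set is named for the Cosmos sniper reload rather than for melee.
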SoundCue /CosmosGameplay/Items/CosmosRifle/Audio/Gun_Rifle_Cosmo_Melee_Windup_Cue
root=Random_0
nodes [
 {id=Random_0, children [WavePlayer_1,WavePlayer_3,WavePlayer_4], properties {weights [1,1,1]}}
 {id=WavePlayer_1, properties {soundWaveAssetPtr=&/Gun_Rifle_Cosmo_Melee_Windup_01}}
 {id=WavePlayer_3, properties {soundWaveAssetPtr=&/Gun_Rifle_Cosmo_Melee_Windup_03}}
 {id=WavePlayer_4, properties {soundWaveAssetPtr=&/Gun_Rifle_Cosmo_Melee_Windup_02}}
]
properties {
 attenuationSettings=/Game/Sounds/Attenuation/Fort_Weapon_Melee_Attack_Close_Range_Mono_Att
 maxDistance=10058.75
 soundClassObject=&Fort_WeaponMelee_PSM
 volumeMultiplier=1.5
 concurrencySet [&/Conc_Jethro_CosmosSniper_Reload]
}
