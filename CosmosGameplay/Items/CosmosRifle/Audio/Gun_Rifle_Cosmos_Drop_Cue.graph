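# Sound cue for the Cosmos rifle's drop sound, built on shared Fortnite weapon-drop audio settings.
# A Random node feeds from a single wave player, so the same drop recording plays each time.
# It uses the loot-collect sound class, weapon-drop attenuation and a shared stop-oldest weapon-drop concurrency set.
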SoundCue /CosmosGameplay/Items/CosmosRifle/Audio/Gun_Rifle_Cosmos_Drop_Cue
root=Random_0
nodes [
 {id=Random_0, children [WavePlayer_0], properties {weights [1]}}
 {id=WavePlayer_0, properties {soundWaveAssetPtr=&/Gun_Rifle_Cosmo_Drop_01}}
]
properties {
 attenuationSettings=&Fort_Weapon_Drop_Att
 maxDistance=3000
 soundClassObject=&Fort_Loot_Collect
 volumeMultiplier=1
 concurrencyOverrides {resolutionRule=PreventNew}
 concurrencySet [&Weapon_Drop_StopOldest]
}
