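# Sound cue for the first-person firing sound of the Cosmos rifle.
# A single close-range fire wave feeds a Modulator with pitch variation pinned flat, and the cue's volume is boosted.
# It routes to the first-person ranged weapon fire sound class and uses shared first-person projectile attenuation and gunfire concurrency assets.
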SoundCue /CosmosGameplay/Items/CosmosRifle/Audio/Gun_Rifle_Cosmos_Fire_1P_Cue
root=Modulator_1
nodes [
 {id=Modulator_1, children [WavePlayer_1], properties {pitchMax=1, pitchMin=1}}
 {id=WavePlayer_1, properties {soundWaveAssetPtr=&/Gun_Rifle_Cosmo_Fire_Close_01}}
]
properties {
 attenuationSettings=&Fort_Weapon_Projectile_FirstPerson_Athena_Att
 maxDistance=2097152
 soundClassObject=&Fort_WeaponRanged_Fire_1P_PSM
 volumeMultiplier=1.5
 attenuationOverrides {falloffDistance=5000, attenuationShapeExtents {x=1000}}
 concurrencyOverrides {bLimitToOwner=true, maxCount=4, resolutionRule=StopOldest}
 concurrencySet [&Weapon_Gunfire_FirstPerson_Conc]
}
