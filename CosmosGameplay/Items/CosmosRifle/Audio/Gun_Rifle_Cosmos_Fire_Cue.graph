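# Sound cue for the Cosmos Rifle's gunfire, layered into close, medium and distant versions.
# A mixer blends the three branches, with medium and distant at reduced input volume.
# Each branch randomly picks one of two waves, then passes through a modulator, range-specific attenuation and a sound class override: 3P for close, distant for the other two.
# Local and global gunfire concurrency sets are assigned, and the debug flag is on.
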SoundCue /CosmosGameplay/Items/CosmosRifle/Audio/Gun_Rifle_Cosmos_Fire_Cue
root=Mixer_38
nodes [
 {
  id=Attenuation_118
  children [Modulator_0]
  properties {attenuationSettings=&Fort_Weapon_Projectile_Close_Athena_Att}
 }
 {
  id=Attenuation_119
  children [Modulator_1]
  properties {attenuationSettings=&Fort_Weapon_Projectile_Medium_Mono_Athena_Att}
 }
 {
  id=Attenuation_120
  children [Modulator_12]
  properties {attenuationSettings=&Fort_Weapon_Projectile_Distant_Mono_Athena_Att}
 }
 {id=Mixer_38, children [SoundClass_77, SoundClass_76, SoundClass_0], properties {inputVolume [1,0.7,0.7]}}
 {id=Modulator_0, children [Random_0], properties {pitchMax=1, pitchMin=1}}
 {id=Modulator_1, children [Random_6], properties {pitchMax=1, pitchMin=1, volumeMax=1, volumeMin=1}}
 {id=Modulator_12, children [Random_3], properties {pitchMax=1, pitchMin=1, volumeMax=0.9, volumeMin=0.9}}
 {id=Random_0, children [WavePlayer_1,WavePlayer_2], properties {weights [1,1]}}
 {id=Random_3, children [WavePlayer_4,WavePlayer_5], properties {weights [1,1]}}
 {id=Random_6, children [WavePlayer_6,WavePlayer_0], properties {weights [1,1]}}
 {
  id=SoundClass_0
  children [Attenuation_120]
  properties {soundClassOverride=&Fort_WeaponRanged_Fire_Distant}
 }
 {
  id=SoundClass_76
  children [Attenuation_119]
  properties {soundClassOverride=&Fort_WeaponRanged_Fire_Distant}
 }
 {
  id=SoundClass_77
  children [Attenuation_118]
  properties {soundClassOverride=&Fort_WeaponRanged_Fire_3P_PSM}
 }
 {id=WavePlayer_0, properties {soundWaveAssetPtr=&/Gun_Rifle_Cosmo_Fire_Med_02}}
 {id=WavePlayer_1, properties {soundWaveAssetPtr=&/Gun_Rifle_Cosmo_Fire_Close_01}}
 {id=WavePlayer_2, properties {soundWaveAssetPtr=&/Gun_Rifle_Cosmo_Fire_Close_02}}
 {id=WavePlayer_4, properties {soundWaveAssetPtr=&/Gun_Rifle_Cosmo_Fire_Dist_01}}
 {id=WavePlayer_5, properties {soundWaveAssetPtr=&/Gun_Rifle_Cosmo_Fire_Dist_02}}
 {id=WavePlayer_6, properties {soundWaveAssetPtr=&/Gun_Rifle_Cosmo_Fire_Med_01}}
]
properties {
 bDebug=true
 maxDistance=27500
 volumeMultiplier=1
 concurrencyOverrides {resolutionRule=PreventNew}
 concurrencySet [&Weapon_Gunfire_Local_Conc, &Weapon_Gunfire_Global_Conc]
}
