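# Sound cue for the Cosmos Rifle's bullet impacts on players.
# A Random node picks one of three equally weighted impact waves and feeds it into a Modulator that varies volume.
# It plays in the Fort_Impacts sound class with the shared bullet impact attenuation and a boosted cue volume.
# A concurrency override caps simultaneous instances at three, stopping the oldest.
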SoundCue /CosmosGameplay/Items/CosmosRifle/Audio/Gun_Rifle_Cosmos_Impt_Player_Cue
root=Modulator_0
nodes [
 {id=Modulator_0, children [Random_0], properties {volumeMax=1.5}}
 {id=Random_0, children [WavePlayer_4,WavePlayer_5,WavePlayer_6], properties {weights [1,1,1]}}
 {id=WavePlayer_4, properties {soundWaveAssetPtr=&/Gun_Rifle_Cosmo_Bullet_Impact_Player_01}}
 {id=WavePlayer_5, properties {soundWaveAssetPtr=&/Gun_Rifle_Cosmo_Bullet_Impact_Player_02}}
 {id=WavePlayer_6, properties {soundWaveAssetPtr=&/Gun_Rifle_Cosmo_Bullet_Impact_Player_03}}
]
properties {
 attenuationSettings=/Game/Sounds/Attenuation/Fort_Bullet_Impact_Att
 bOverrideConcurrency=true
 maxDistance=5300
 soundClassObject=&Fort_Impacts
 volumeMultiplier=1.5
 concurrencyOverrides {maxCount=3, resolutionRule=StopOldest}
}
