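# Sound cue for Cosmos rifle bullet impacts on water, classed as ranged weapon audio and spatialized with the shared bullet impact attenuation.
# A mixer layers a random Cosmos bullet whiz-by, at reduced volume with slight upward pitch variation, with a random medium water impact.
# The water impact plays after a short fixed delay, so the whiz-by leads into it.
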SoundCue /CosmosGameplay/Items/CosmosRifle/Audio/Gun_Rifle_Cosmos_Impt_Water_Cue
root=Mixer_0
nodes [
 {id=Delay_0, children [Random_2], properties {delayMax=0.2, delayMin=0.2}}
 {id=Mixer_0, children [Delay_0,Modulator_1], properties {inputVolume [1,0.7]}}
 {id=Modulator_1, children [Random_3], properties {pitchMax=1.15, pitchMin=1}}
 {id=Random_2, children [WavePlayer_6, WavePlayer_8, WavePlayer_9, WavePlayer_11], properties {weights [1,1,1,1]}}
 {id=Random_3, children [WavePlayer_0, WavePlayer_1, WavePlayer_15, WavePlayer_16], properties {weights [1,1,1,1]}}
 {id=WavePlayer_0, properties {soundWaveAssetPtr=&/Gun_Rifle_Cosmo_Bullet_WhizBy_04}}
 {id=WavePlayer_1, properties {soundWaveAssetPtr=&/Gun_Rifle_Cosmo_Bullet_WhizBy_01}}
 {id=WavePlayer_11, properties {soundWaveAssetPtr=/Game/Sounds/Fort_Weapons/Impacts/Bullet_Impacts_S10/Bullet_Impact_Water_Med_06}}
 {id=WavePlayer_15, properties {soundWaveAssetPtr=&/Gun_Rifle_Cosmo_Bullet_WhizBy_02}}
 {id=WavePlayer_16, properties {soundWaveAssetPtr=&/Gun_Rifle_Cosmo_Bullet_WhizBy_03}}
 {id=WavePlayer_6, properties {soundWaveAssetPtr=/Game/Sounds/Fort_Weapons/Impacts/Bullet_Impacts_S10/Bullet_Impact_Water_Med_01}}
 {id=WavePlayer_8, properties {soundWaveAssetPtr=/Game/Sounds/Fort_Weapons/Impacts/Bullet_Impacts_S10/Bullet_Impact_Water_Med_03}}
 {id=WavePlayer_9, properties {soundWaveAssetPtr=/Game/Sounds/Fort_Weapons/Impacts/Bullet_Impacts_S10/Bullet_Impact_Water_Med_04}}
]
properties {
 attenuationSettings=/Game/Sounds/Attenuation/Fort_Bullet_Impact_Att
 maxDistance=5300
 soundClassObject=&Fort_WeaponRanged
 volumeMultiplier=1
}
