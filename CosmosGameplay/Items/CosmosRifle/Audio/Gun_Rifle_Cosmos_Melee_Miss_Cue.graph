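# Sound cue for a missed melee attack with the Cosmos rifle, playing a whoosh.
# A Random node picks one of three equally weighted medium stereo whoosh waves and feeds a Modulator at default settings.
# It uses close-range melee attenuation and the weapon melee sound class.
# The cue's volume multiplier is zero, so it is effectively silent as configured.
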SoundCue /CosmosGameplay/Items/CosmosRifle/Audio/Gun_Rifle_Cosmos_Melee_Miss_Cue
root=Modulator_0
nodes [
 {id=Modulator_0, children [Random_0]}
 {id=Random_0, children [WavePlayer_1,WavePlayer_2,WavePlayer_3], properties {weights [1,1,1]}}
 {id=WavePlayer_1, properties {soundWaveAssetPtr=/Game/Sounds/Fort_Weapons/Whooshes/Whoosh_Medium_Stereo_02}}
 {id=WavePlayer_2, properties {soundWaveAssetPtr=/Game/Sounds/Fort_Weapons/Whooshes/Whoosh_Medium_Stereo_05}}
 {id=WavePlayer_3, properties {soundWaveAssetPtr=/Game/Sounds/Fort_Weapons/Whooshes/Whoosh_Medium_Stereo_03}}
]
properties {
 attenuationSettings=/Game/Sounds/Attenuation/Fort_Weapon_Melee_Attack_Close_Range_Mono_Att
 maxDistance=10058.75
 soundClassObject=&Fort_WeaponMelee_PSM
 volumeMultiplier=0
}
