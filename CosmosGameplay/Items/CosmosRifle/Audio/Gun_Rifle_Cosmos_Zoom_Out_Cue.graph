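# Sound Cue for the Cosmos rifle's zoom-out, played as weapon foley when aiming down sights ends.
# A single wave, Gun_Rifle_Cosmo_ADS_End_01, feeds a default Modulator node for slight playback variation.
# The cue uses the weapon foley sound class, a slightly reduced volume, and overrides for attenuation shape and falloff.
# Its concurrency override prevents new instances once the limit is reached.
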SoundCue /CosmosGameplay/Items/CosmosRifle/Audio/Gun_Rifle_Cosmos_Zoom_Out_Cue
root=Modulator_1
nodes [
 {id=Modulator_1, children [WavePlayer_0]}
 {id=WavePlayer_0, properties {soundWaveAssetPtr=&/Gun_Rifle_Cosmo_ADS_End_01}}
]
properties {
 attenuationSettings=&Fort_Weapon_Reload_Att
 bOverrideAttenuation=true
 bOverrideConcurrency=true
 maxDistance=2700
 soundClassObject=&Fort_WeaponFoley
 volumeMultiplier=0.752
 attenuationOverrides {falloffDistance=1500, nonSpatializedRadiusStart=200, attenuationShapeExtents {x=1200}}
 concurrencyOverrides {resolutionRule=PreventNew}
}
