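# Sound cue for the start of the Jethro Cosmos sniper rifle's reload.
# A Random node picks one of two reload-start wave variations with equal weight.
# The cue plays in the weapon foley sound class at reduced volume, limited by a concurrency asset for this weapon's reload.
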SoundCue /CosmosGameplay/Items/CosmosRifle/Audio/SC_Jethro_CosmosSniper_Reload_Start
root=Random_0
nodes [
 {id=Random_0, children [WavePlayer_0,WavePlayer_1], properties {weights [1,1]}}
 {id=WavePlayer_0, properties {soundWaveAssetPtr=&/Jethro_Comsos_ReloadStart_02}}
 {id=WavePlayer_1, properties {soundWaveAssetPtr=&/Jethro_Comsos_ReloadStart_01}}
]
properties {
 maxDistance=2097152
 soundClassObject=&Fort_WeaponFoley
 volumeMultiplier=0.6
 concurrencySet [&/Conc_Jethro_CosmosSniper_Reload]
}
