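; Cosmos sniper rifle Blueprint that adds a thermal scope effect and scoped-only out-of-ammo audio.
; Targeting changes reach the parent only with a valid local viewing controller, also adjusting significance-based custom depth rendering.
; After a short delay, scope post-process shows the mask and drives thermal material collection parameters, using a timeline for scale.
; Scoped-audio events track scope state, gating the empty sound to avoid melee triggers and resetting thermal effects on exit.
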(event EventOnSetTargeting
  (params (bNewIsTargeting :type Boolean))
  ;[1 1 1] Cosmos Scope Post Process setup
  (Networking|CanExecuteCosmeticEvents
    (:True
      (bind _asfort_pawn (Utilities|Casting|CastToFortPawn (Game|GetInstigator)))
      (Utilities|IsValid (Pawn|GetLocalViewingPlayerController _asfort_pawn)
        (:Is%20Valid
          (|CallParentFunction bNewIsTargeting)
          (bind _asfort_pawn_2 (Utilities|Casting|CastToFortPawn (Game|GetInstigator)))
          (bind _asfort_player_controller (Utilities|Casting|CastToFortPlayerController (Pawn|GetLocalViewingPlayerController _asfort_pawn_2)))
          (Rendering|SetIgnoreSignifanceBasedCustomDepthRendering _asfort_player_controller (Camera|GetIsTargeting) 0))))))

(event EventPlayScopePP
  ;[1 1 1] Cosmos Scope Post Process setup
  (Utilities|FlowControl|Delay 0.1)
  (if (Camera|GetIsTargeting)
    (Rendering|SetVisibility (Variables|Default|GetSM_Jethro_Cosmos_ScopeMask) true false)
    (CallFunction|SetPostProcessParams 1.0)
    (Components|Timeline|SetPlayRate (Variables|B_Rifle_Sniper_Cosmos|GetScalePlayerThermal) (/ 1.0 (Variables|Default|GetBlendInTime)))
    (Rendering|Material|SetScalarParameterValue "/Game/Athena/Items/Weapons/Abilities/Thermal/MPC_ThermalGlobal.MPC_ThermalGlobal" "ThermalFadeout" 1.0 :!class /Script/Engine.KismetMaterialLibrary)
    (Utilities|FlowControl|DoOnce false
      (:Completed
        (bind (_direction _newtrack_0) (Timeline|ScalePlayerThermal PlayFromStart)
          (:Update
            _)
          (:Finished
            _))
        (Rendering|Material|SetScalarParameterValue "/Game/Athena/Items/Weapons/Abilities/Thermal/MPC_ThermalGlobal.MPC_ThermalGlobal" "ThermalScale" _newtrack_0 :!class /Script/Engine.KismetMaterialLibrary)))))

(event Custom|ResetThermalScopeParams
  ; Reset thermal params.
  (Rendering|SetVisibility (Variables|Default|GetSM_Jethro_Cosmos_ScopeMask) false false)
  (CallFunction|SetPostProcessParams 0.0)
  (Rendering|Material|SetScalarParameterValue "/Game/Athena/Items/Weapons/Abilities/Thermal/MPC_ThermalGlobal.MPC_ThermalGlobal" "ThermalFadeout" :!class /Script/Engine.KismetMaterialLibrary)
  (Rendering|Material|SetScalarParameterValue "/Game/Athena/Items/Weapons/Abilities/Thermal/MPC_ThermalGlobal.MPC_ThermalGlobal" "ThermalScale" :!class /Script/Engine.KismetMaterialLibrary))

(event EventOnOutOfAmmoFire
  ;[1 1 1] Out of Ammo Sound - Play this sound instead of normal hook to prevent "out of ammo" sound from playing during melee
  (if (Variables|Default|GetIsScoped)
    (Audio|PlaySound2D (Variables|Sound|GetNoAmmoSound))))

(event Audio|EventStartScopedAudio
  ;[1 1 1] Out of Ammo Sound - Play this sound instead of normal hook to prevent "out of ammo" sound from playing during melee
  (|CallParentFunction)
  (Variables|Default|SetIsScoped true))

(event Audio|EventStopScopedAudio
  ;[1 1 1] Out of Ammo Sound - Play this sound instead of normal hook to prevent "out of ammo" sound from playing during melee
  (|CallParentFunction)
  (Variables|Default|SetIsScoped false)
  (CallFunction|ResetThermalScopeParams))

(fn ConstructionScript
  (|CallParentFunction))
